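# Sound cue for the FlexLegend lever shotgun's aim-down-sights-in foley sound.
# A Random node picks one of three equally weighted ADS-in wave variations and is excluded from branch culling.
# The cue plays under the Fort weapon foley sound class with a boosted volume and borrows the weapon reload attenuation settings for spatial falloff.
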SoundCue /FlexLegendWeaponGameplay/Gameplay/LeverShotgun/Sounds/FlexLegend_LeverShotgun_Weapon_ADS_In_Cue
root=Random_1
nodes [
 {id=Random_1, children [WavePlayer_0,WavePlayer_1,WavePlayer_2], properties {bShouldExcludeFromBranchCulling=true, weights [1,1,1]}}
 {
  id=WavePlayer_0
  properties {soundWaveAssetPtr=/FlexLegendWeaponGameplay/Gameplay/LeverShotgun/Sounds/Waves/LeverShotgun_FlexLegend_ADS_In_01}
 }
 {
  id=WavePlayer_1
  properties {soundWaveAssetPtr=/FlexLegendWeaponGameplay/Gameplay/LeverShotgun/Sounds/Waves/LeverShotgun_FlexLegend_ADS_In_02}
 }
 {
  id=WavePlayer_2
  properties {soundWaveAssetPtr=/FlexLegendWeaponGameplay/Gameplay/LeverShotgun/Sounds/Waves/LeverShotgun_FlexLegend_ADS_In_03}
 }
]
properties {
 attenuationSettings=&Fort_Weapon_Reload_Att
 maxDistance=2250
 soundClassObject=&Fort_WeaponFoley
 volumeMultiplier=1.65
}
