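# Sound cue for dropping the FlexLegend lever shotgun weapon.
# A Random node picks one of three drop wave variations with equal weighting and feeds it straight to the output.
# It plays in the weapon foley sound class and borrows the shared weapon reload attenuation for distance falloff.
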SoundCue /FlexLegendWeaponGameplay/Gameplay/LeverShotgun/Sounds/FlexLegend_LeverShotgun_Weapon_Drop_Cue
root=Random_0
nodes [
 {id=Random_0, children [WavePlayer_0,WavePlayer_1,WavePlayer_2], properties {weights [1,1,1]}}
 {id=WavePlayer_0, properties {soundWaveAssetPtr=/FlexLegendWeaponGameplay/Gameplay/LeverShotgun/Sounds/Waves/LeverShotgun_FlexLegend_Drop_01}}
 {id=WavePlayer_1, properties {soundWaveAssetPtr=/FlexLegendWeaponGameplay/Gameplay/LeverShotgun/Sounds/Waves/LeverShotgun_FlexLegend_Drop_02}}
 {id=WavePlayer_2, properties {soundWaveAssetPtr=/FlexLegendWeaponGameplay/Gameplay/LeverShotgun/Sounds/Waves/LeverShotgun_FlexLegend_Drop_03}}
]
properties {
 attenuationSettings=&Fort_Weapon_Reload_Att
 maxDistance=2250
 soundClassObject=&Fort_WeaponFoley
}
