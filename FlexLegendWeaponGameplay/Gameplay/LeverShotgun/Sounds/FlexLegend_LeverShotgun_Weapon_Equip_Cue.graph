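# Sound cue for equipping the FlexLegend lever shotgun, layering an equip foley sound with a delayed lever action.
# A Mixer plays the single equip wave alongside a branch held back by a short fixed delay.
# That branch passes one of six equally weighted, randomly picked reload-end waves through a Modulator, reusing reload audio for variation.
# It uses the weapon foley sound class and reload attenuation.
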SoundCue /FlexLegendWeaponGameplay/Gameplay/LeverShotgun/Sounds/FlexLegend_LeverShotgun_Weapon_Equip_Cue
root=Mixer_0
nodes [
 {id=Delay_0, children [Modulator_0], properties {delayMax=0.15, delayMin=0.15}}
 {id=Mixer_0, children [WavePlayer_0,Delay_0], properties {inputVolume [0.5,0.8]}}
 {id=Modulator_0, children [Random_0]}
 {
  id=Random_0
  children [WavePlayer_3,WavePlayer_4,WavePlayer_5,WavePlayer_6,WavePlayer_7,WavePlayer_8]
  properties {bShouldExcludeFromBranchCulling=true, weights [1,1,1,1,1,1]}
 }
 {
  id=WavePlayer_0
  properties {soundWaveAssetPtr=/FlexLegendWeaponGameplay/Gameplay/LeverShotgun/Sounds/Waves/LeverShotgun_FlexLegend_Equip_01}
 }
 {
  id=WavePlayer_3
  properties {soundWaveAssetPtr=/FlexLegendWeaponGameplay/Gameplay/LeverShotgun/Sounds/Waves/LeverShotgun_FlexLegend_Reload_End_01}
 }
 {
  id=WavePlayer_4
  properties {soundWaveAssetPtr=/FlexLegendWeaponGameplay/Gameplay/LeverShotgun/Sounds/Waves/LeverShotgun_FlexLegend_Reload_End_02}
 }
 {
  id=WavePlayer_5
  properties {soundWaveAssetPtr=/FlexLegendWeaponGameplay/Gameplay/LeverShotgun/Sounds/Waves/LeverShotgun_FlexLegend_Reload_End_03}
 }
 {
  id=WavePlayer_6
  properties {soundWaveAssetPtr=/FlexLegendWeaponGameplay/Gameplay/LeverShotgun/Sounds/Waves/LeverShotgun_FlexLegend_Reload_End_04}
 }
 {
  id=WavePlayer_7
  properties {soundWaveAssetPtr=/FlexLegendWeaponGameplay/Gameplay/LeverShotgun/Sounds/Waves/LeverShotgun_FlexLegend_Reload_End_05}
 }
 {
  id=WavePlayer_8
  properties {soundWaveAssetPtr=/FlexLegendWeaponGameplay/Gameplay/LeverShotgun/Sounds/Waves/LeverShotgun_FlexLegend_Reload_End_06}
 }
]
properties {
 attenuationSettings=&Fort_Weapon_Reload_Att
 maxDistance=2250
 soundClassObject=&Fort_WeaponFoley
}
