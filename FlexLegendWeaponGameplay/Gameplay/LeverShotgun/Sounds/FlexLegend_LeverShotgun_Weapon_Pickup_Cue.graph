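# Sound cue for picking up the FlexLegend lever shotgun weapon.
# A Random node picks one of three equally weighted pickup wave variations and feeds it straight to the output.
# It uses the shared Fort weapon pickup attenuation and loot collect sound class, with a slight volume boost.
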SoundCue /FlexLegendWeaponGameplay/Gameplay/LeverShotgun/Sounds/FlexLegend_LeverShotgun_Weapon_Pickup_Cue
root=Random_0
nodes [
 {id=Random_0, children [WavePlayer_0,WavePlayer_1,WavePlayer_2], properties {weights [1,1,1]}}
 {
  id=WavePlayer_0
  properties {soundWaveAssetPtr=/FlexLegendWeaponGameplay/Gameplay/LeverShotgun/Sounds/Waves/LeverShotgun_FlexLegend_Pickup_01}
 }
 {
  id=WavePlayer_1
  properties {soundWaveAssetPtr=/FlexLegendWeaponGameplay/Gameplay/LeverShotgun/Sounds/Waves/LeverShotgun_FlexLegend_Pickup_02}
 }
 {
  id=WavePlayer_2
  properties {soundWaveAssetPtr=/FlexLegendWeaponGameplay/Gameplay/LeverShotgun/Sounds/Waves/LeverShotgun_FlexLegend_Pickup_03}
 }
]
properties {
 attenuationSettings=&Fort_Weapon_Pickup_Att
 maxDistance=3500
 soundClassObject=&Fort_Loot_Collect
 volumeMultiplier=1.1538
}
