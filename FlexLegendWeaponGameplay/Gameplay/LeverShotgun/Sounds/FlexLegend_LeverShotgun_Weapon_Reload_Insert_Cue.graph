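# Sound cue for the shell-insert step of reloading the FlexLegend lever shotgun.
# A Random node picks one of six equally weighted insert wave variations and is excluded from branch culling.
# A Modulator at the root sits over the random pick to apply pitch and volume variation at its default settings.
# Playback uses the shared Fortnite weapon reload attenuation and the weapon foley sound class.
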SoundCue /FlexLegendWeaponGameplay/Gameplay/LeverShotgun/Sounds/FlexLegend_LeverShotgun_Weapon_Reload_Insert_Cue
root=Modulator_0
nodes [
 {id=Modulator_0, children [Random_0]}
 {
  id=Random_0
  children [WavePlayer_0,WavePlayer_1,WavePlayer_2,WavePlayer_3,WavePlayer_4,WavePlayer_5]
  properties {bShouldExcludeFromBranchCulling=true, weights [1,1,1,1,1,1]}
 }
 {
  id=WavePlayer_0
  properties {soundWaveAssetPtr=/FlexLegendWeaponGameplay/Gameplay/LeverShotgun/Sounds/Waves/LeverShotgun_FlexLegend_Reload_Insert_01}
 }
 {
  id=WavePlayer_1
  properties {soundWaveAssetPtr=/FlexLegendWeaponGameplay/Gameplay/LeverShotgun/Sounds/Waves/LeverShotgun_FlexLegend_Reload_Insert_02}
 }
 {
  id=WavePlayer_2
  properties {soundWaveAssetPtr=/FlexLegendWeaponGameplay/Gameplay/LeverShotgun/Sounds/Waves/LeverShotgun_FlexLegend_Reload_Insert_03}
 }
 {
  id=WavePlayer_3
  properties {soundWaveAssetPtr=/FlexLegendWeaponGameplay/Gameplay/LeverShotgun/Sounds/Waves/LeverShotgun_FlexLegend_Reload_Insert_04}
 }
 {
  id=WavePlayer_4
  properties {soundWaveAssetPtr=/FlexLegendWeaponGameplay/Gameplay/LeverShotgun/Sounds/Waves/LeverShotgun_FlexLegend_Reload_Insert_05}
 }
 {
  id=WavePlayer_5
  properties {soundWaveAssetPtr=/FlexLegendWeaponGameplay/Gameplay/LeverShotgun/Sounds/Waves/LeverShotgun_FlexLegend_Reload_Insert_06}
 }
]
properties {
 attenuationSettings=&Fort_Weapon_Reload_Att
 maxDistance=2250
 soundClassObject=&Fort_WeaponFoley
 volumeMultiplier=1
}
